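# Translucent UI material that draws a tintable, rotatable arrow indicator for the player augments UI.
# It rotates the UVs of a shared arrow texture with CustomRotator, driven by the Rotation parameter.
# The green channel blends Color_Glow into Color_Arrow, and the saturated blue channel times Alpha forms an opacity mask.
# BorderAmount has no visible effect because its lerp takes the green channel on both inputs.
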Material /PlayerAugmentsUI/Materials/M_UI_AugmentArrow_Indicator
graph {
 properties {blendMode=BLEND_Translucent, materialDomain=MD_UI, pixelDepthOffsetMode=PDOM_Legacy, usedShadingModels=Default Lit}
 nodes [
  n1 = TextureSample(
   UVs=MaterialFunctionCall(UVs=TextureCoordinate(), "Rotation Angle (0-1)"=ScalarParameter() {defaultValue=0.25, parameterName=Rotation}) {materialFunction=&CustomRotator}.Rotated Values
   Tex=TextureObject() {texture=/FortniteCoreUI/Frontend/Spatial/Art/T_Ui_arrow}
  )
  LinearInterpolate(
   A=VectorParameter() {parameterName=Color_Glow, defaultValue=<rgba 0 1 0 1>}.RGB
   B=VectorParameter() {parameterName=Color_Arrow, defaultValue=<rgba 0 0 1 1>}.RGB
   Alpha=LinearInterpolate(A=@n1.G, B=@n1.G, Alpha=ScalarParameter() {defaultValue=0.25, parameterName=BorderAmount})
  )
  Saturate(None=Multiply(A=Saturate(None=@n1.B), B=ScalarParameter() {defaultValue=1, parameterName=Alpha}))
 ]
}
