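# Translucent UI master material that draws a radial progress ring, named for the augment timer.
# An angular mask from the animatable Percent separates fill from backing, while two radial gradients cut the ring with adjustable widths and edge hardness.
# Fill and backing each blend between two colors, with the backing optionally auto-pulsing through a time-driven, sine-eased ping-pong.
# Opacity combines separate ring and progress alphas, the ring mask and a final alpha multiplier.
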Material /PlayerAugmentsUI/Materials/M_UI_AugmentTimer_Master
graph {
 properties {blendMode=BLEND_Translucent, materialDomain=MD_UI, pixelDepthOffsetMode=PDOM_Legacy, usedShadingModels=Default Lit}
 nodes [
  n1 = ScalarParameter() {parameterName=Animate_BackingColor}
  n2 = Time()
  n3 = ScalarParameter() {defaultValue=2, parameterName=AutoAnimate_BackingColorDuration}
  n4 = Subtract(A=Divide(A=Fmod(A=@n2, B=@n3), B=Divide(A=@n3)))
  n5 = LinearInterpolate(
   B=MaterialFunctionCall("Value (0-1)"=Abs(None=@n4)) {materialFunction=&MF_UI_EaseCurves}.EaseInOutSine
   Alpha=ScalarParameter() {parameterName=AutoAnimate_BackingColor}
  )
  n6 = Power(Base=NamedRerouteUsage(), Exp=ScalarParameter() {defaultValue=1, parameterName=ColorChangeRateAdjustment})
  LinearInterpolate(
   A=LinearInterpolate(
    A=VectorParameter() {parameterName=Backing_Color, defaultValue=<rgba 0 1 0 1>}.RGB
    B=VectorParameter() {parameterName=Backing_Color_Stop, defaultValue=<rgba 0 1 0 1>}.RGB
    Alpha=Max(A=@n1, B=@n5)
   )
   B=LinearInterpolate(
    A=VectorParameter() {parameterName=Fill_Color_Start, defaultValue=<rgba 1 0 0 1>}.RGB
    B=VectorParameter() {parameterName=Fill_Color_Stop, defaultValue=<rgba 1 0 0 1>}.RGB
    Alpha=Saturate(None=@n6)
   )
   Alpha=NamedRerouteUsage()
  )
  n7 = MaterialFunctionCall!(Vector or UVs) {materialFunction=/FortUILibrary/MaterialFunctions/Basics/MF_UI_CheaperVectorToRadial}
  n8 = OneMinus(None=@n7.Vector Converted to Angle)
  n9 = NamedRerouteDeclaration(None=LinearInterpolate(
    A=ScalarParameter() {defaultValue=0.5, group=ProgressBar, parameterName=Percent_Start}
    B=ScalarParameter() {defaultValue=0.5, group=ProgressBar, parameterName=Percent_Stop}
    Alpha=ScalarParameter() {parameterName=AnimatePercentChange}
   )) {name=Percent, nodeColor=<rgba 1 0 0.0703 1>}
  n10 = ComponentMask(None=@n7.Radial Coordinates) {g=true}
  n11 = ScalarParameter() {defaultValue=512, group=ProgressBar, parameterName=ProgressBar_EdgeHardness}
  n12 = Saturate(None=Multiply(A=Subtract(A=@n8, B=@n9), B=Multiply(A=@n10, B=@n11)))
  NamedRerouteDeclaration(None=OneMinus(None=Saturate(None=@n12))) {name=ProgressAlpha, nodeColor=<rgba 0.7578 1 0 1>}
  n13 = ScalarParameter() {parameterName=Ring_OutsideWidth}
  n14 = Saturate(None=Multiply(
    A=MaterialFunctionCall(GradientStart=@n13) {materialFunction=&MF_UI_RadialGradient}.Smooth
    B=ScalarParameter() {defaultValue=64, parameterName=Ring_OutsideHardness}
   ))
  n15 = ScalarParameter() {defaultValue=0.5, parameterName=Ring_InsideWidth}
  n16 = MaterialFunctionCall(GradientStart=@n15) {materialFunction=&MF_UI_RadialGradient}
  n17 = ScalarParameter() {defaultValue=6, parameterName=Ring_InsideHardness}
  n18 = OneMinus(None=Saturate(None=Multiply(A=@n16.Smooth, B=@n17)))
  NamedRerouteDeclaration(None=Saturate(None=DotProduct(A=@n14, B=@n18))) {name=RingAlpha, nodeColor=<rgba 0.9922 1 0 1>}
  n19 = LinearInterpolate(
   A=LinearInterpolate(
    A=ScalarParameter() {defaultValue=0.5, parameterName=Ring_Alpha}
    B=ScalarParameter() {defaultValue=0.5, parameterName=Ring_Alpha_Stop}
    Alpha=ScalarParameter() {parameterName=Animate_Ring_Alpha}
   )
   B=ScalarParameter() {defaultValue=1, parameterName=Progress_Alpha}
   Alpha=NamedRerouteUsage()
  )
  n20 = NamedRerouteUsage()
  NamedRerouteDeclaration(None=Multiply(A=DotProduct(A=@n19, B=@n20), B=ScalarParameter() {defaultValue=1, parameterName=Final_Alpha})) {name=Final Alpha, nodeColor=<rgba 0 0.8359 1 1>}
  OneMinus(None=@n15)
 ]
}
