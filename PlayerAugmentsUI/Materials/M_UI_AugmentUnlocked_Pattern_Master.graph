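# UI master material for an animated, translucent tile pattern, named for the augment-unlocked display.
# Panning UVs are quantized with Ceil into cells, whose vector length offsets a time-driven cosine so squares pulse at different rates.
# An optional looping sine swipe modulates that mask, which blends ColorA to ColorB and drives opacity.
# Opacity adds a lift for dark squares and optional sine-based left/right and top/bottom edge fades before a final alpha multiplier.
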Material /PlayerAugmentsUI/Materials/M_UI_AugmentUnlocked_Pattern_Master
graph {
 properties {blendMode=BLEND_Translucent, materialDomain=MD_UI, pixelDepthOffsetMode=PDOM_Legacy, usedShadingModels=Default Lit}
 nodes [
  n1 = MaterialFunctionCall() {materialFunction=&GetUserInterfaceUV}
  n2 = Constant() {r=1}
  n3 = MaterialFunctionCall(Float2=@n1.Pixel Size) {materialFunction=&BreakOutFloat2Components}
  n4 = Max(A=Divide(A=@n3.G, B=@n3.G)) {constB=0.001}
  n5 = OneMinus(None=@n4)
  n6 = Add(A=Multiply(A=OneMinus(None=@n1.Normalized UV), B=AppendVector(A=@n2, B=@n4)), B=AppendVector(A=Constant(), B=Multiply(A=@n5) {constB=0.5}))
  n7 = AppendVector(
   A=ScalarParameter() {defaultValue=0.25, group=Tiling, parameterName=PanSpeed_SolidTiles_X}
   B=ScalarParameter() {group=Tiling, parameterName=PanSpeed_SolidTiles_Y}
  )
  n8 = Multiply(A=Frac(None=Panner(Coordinate=@n6, Speed=@n7) {speedX=0.5}), B=ScalarParameter() {defaultValue=2, group=Tiling, parameterName=Tile})
  n9 = Multiply(
   A=MaterialFunctionCall(Vector 2=Ceil(None=@n8)) {materialFunction=/Engine/Functions/Engine_MaterialFunctions02/Utility/VectorLength}.V2 Length
   B=Multiply(A=Time(), B=ScalarParameter() {defaultValue=0.125, group=Tiling, parameterName=Tile_AnimSpeed})
  )
  n10 = Saturate(None=LinearInterpolate(A=ScalarParameter() {defaultValue=0.125, group=Tiling, parameterName=MoreConsistentTiles}, Alpha=Cosine(None=@n9)))
  n11 = ScalarParameter() {defaultValue=-1, parameterName=PanPosition_Start}
  n12 = ScalarParameter() {defaultValue=1, parameterName=PanPosition_Stop}
  n13 = Fmod(
   A=Multiply(A=Time(), B=ScalarParameter() {defaultValue=0.75, parameterName=PanSpeed})
   B=ScalarParameter() {defaultValue=1, parameterName=FmodStop}
  )
  n14 = Add(A=TextureCoordinate(), B=OneMinus(None=LinearInterpolate(A=@n11, B=@n12, Alpha=@n13)))
  n15 = DotProduct(A=@n10, B=ComponentMask(None=Sine(None=@n14) {period=2}) {r=true})
  n16 = Saturate(None=LinearInterpolate(A=@n10, B=Saturate(None=@n15), Alpha=ScalarParameter() {defaultValue=0.5, parameterName=UseSwipe}))
  LinearInterpolate(
   A=VectorParameter() {parameterName=ColorA, defaultValue=<rgba 0 1 0 1>}.RGB
   B=VectorParameter() {parameterName=ColorB, defaultValue=<rgba 1 0 0 1>}.RGB
   Alpha=@n16
  )
  n17 = LinearInterpolate(A=ScalarParameter() {parameterName=LiftNegativeSquares}, Alpha=@n16)
  n18 = Multiply(A=ComponentMask(None=TextureCoordinate()) {r=true}) {constB=0.5}
  n19 = ScalarParameter() {defaultValue=1, parameterName=FadeWidth_LeftRight}
  n20 = Multiply(A=Subtract(A=Sine(None=@n18), B=OneMinus(None=@n19)) {constB=0.25}, B=Divide(B=@n19) {constB=0.5})
  n21 = ScalarParameter() {defaultValue=1, parameterName=FadeDensity_LeftRight}
  n22 = Power(Base=Saturate(None=Multiply(A=@n20, B=@n21)), Exp=ScalarParameter() {defaultValue=1, parameterName=FadeRate_LeftRight})
  n23 = LinearInterpolate(A=@n17, B=Saturate(None=DotProduct(A=@n17, B=@n22)), Alpha=ScalarParameter() {parameterName=FadeEdgesLeftRight})
  n24 = MaterialFunctionCall(UVs=TextureCoordinate(), "Rotation Angle (0-1)"=Constant() {r=0.25}) {materialFunction=&CustomRotator}
  n25 = Sine(None=Multiply(A=ComponentMask(None=@n24.Rotated Values) {r=true}) {constB=0.5})
  n26 = ScalarParameter() {defaultValue=1, parameterName=FadeWidth_TopBottom}
  n27 = OneMinus(None=@n26)
  n28 = Multiply(
   A=Multiply(A=Subtract(A=@n25, B=@n27) {constB=0.25}, B=Divide(B=@n26) {constB=0.5})
   B=ScalarParameter() {defaultValue=1, parameterName=FadeDensity_TopBottom}
  )
  n29 = DotProduct(A=@n23, B=Power(Base=Saturate(None=@n28), Exp=ScalarParameter() {defaultValue=1, parameterName=FadeRate_TopBottom}))
  Multiply(
   A=LinearInterpolate(A=@n23, B=Saturate(None=@n29), Alpha=ScalarParameter() {parameterName=FadeEdgesTopBottom})
   B=ScalarParameter() {defaultValue=1, parameterName=Final_Alpha}
  )
 ]
}
