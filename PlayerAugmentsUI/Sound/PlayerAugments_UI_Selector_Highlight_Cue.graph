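# Sound cue for the highlight sound of the Player Augments UI selector.
# A single wave player feeds the highlight wave straight to the output at a strongly reduced volume, in the non-critical in-game UI sound class.
# A concurrency override allows only a couple of simultaneous instances and stops the oldest when the limit is exceeded.
# The cue's debug flag is left enabled.
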SoundCue /PlayerAugmentsUI/Sound/PlayerAugments_UI_Selector_Highlight_Cue
root=WavePlayer_2
nodes [{id=WavePlayer_2, properties {soundWaveAssetPtr=&/PlayerAugments_Selector_Highlight_01}}]
properties {
 bDebug=true
 bOverrideConcurrency=true
 maxDistance=2097152
 soundClassObject=&Fort_UI_InGame_NonCritical
 volumeMultiplier=0.1
 concurrencyOverrides {maxCount=2, resolutionRule=StopOldest}
}
