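# Sound cue for the hover tick of the Player Augments UI selector.
# A Random node picks with equal weight between two branches playing identically named center tick waves, one pitched slightly down and one at normal pitch, for subtle variation.
# It plays at reduced volume in the non-critical in-game UI sound class, with a concurrency override that caps it at two voices and stops the oldest.
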SoundCue /PlayerAugmentsUI/Sound/PlayerAugments_UI_Selector_Tick_Cue
root=Random_0
nodes [
 {id=Modulator_1, children [WavePlayer_0], properties {pitchMax=0.95, volumeMax=1, volumeMin=1}}
 {id=Modulator_2, children [WavePlayer_1], properties {pitchMax=1, pitchMin=1, volumeMax=1, volumeMin=1}}
 {id=Random_0, children [Modulator_1,Modulator_2], properties {bShouldExcludeFromBranchCulling=true, weights [1,1]}}
 {id=WavePlayer_0, properties {soundWaveAssetPtr=&/PlayerAugments_Selector_Hover_Tick_Center_01}}
 {id=WavePlayer_1, properties {soundWaveAssetPtr=&PlayerAugments_Selector_Hover_Tick_Center_01}}
]
properties {
 bDebug=true
 bOverrideConcurrency=true
 maxDistance=2097152
 soundClassObject=&Fort_UI_InGame_NonCritical
 volumeMultiplier=0.4
 concurrencyOverrides {maxCount=2, resolutionRule=StopOldest}
}
